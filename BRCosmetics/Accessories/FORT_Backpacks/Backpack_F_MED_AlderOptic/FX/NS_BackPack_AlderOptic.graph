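# Niagara system for the AlderOptic backpack cosmetic, adding pulsing pink glows and a flickering ribbon trail.
# GlowsAndLights and EarGlows place sprites at partition-indexed fixed offsets, with camera-angle-driven fades and sizing on the body glows.
# A six-particle ribbon waves and lags via position interpolation, changing direction and clamps with IsSkydiving, while the system timeline boosts flicker and scroll speed.
# Positions add a gender offset; the debug emitter is disabled.
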NiagaraSystem /BRCosmetics/Accessories/FORT_Backpacks/Backpack_F_MED_AlderOptic/FX/NS_BackPack_AlderOptic
dependencies {
 usedDataInterfaces [
  &NiagaraDataInterfaceCamera
  &NiagaraDataInterfaceArrayFloat
  &NiagaraDataInterfaceCurve
  &NiagaraDataInterfaceSocketReader
 ]
 usedDynamicInputs [
  &SelectFloatFromArray
  &VectorLength
  &SetBoolByFloatComparison
 ]
 usedModules [
  &self:SystemSpawnScript.SystemScriptSource.SystemScriptGraph.NiagaraNodeAssignment_2.SetVariables_#1
  &SystemState
  &self:SystemSpawnScript.SystemScriptSource.SystemScriptGraph.NiagaraNodeAssignment_3.SetVariables_#2
  &Timeline
  &self:SystemSpawnScript.SystemScriptSource.SystemScriptGraph.NiagaraNodeAssignment_4.SetVariables_#3
 ]
}
emitters {
 EarGlows {
  emitterUpdateScript [
   &EmitterState(
    Inactive Response=::NewEnumerator1
    Life Cycle Mode=::NewEnumerator1
    Loop Behavior=::NewEnumerator1
    Loop Duration Mode=::NewEnumerator1
    UseLoopDelay=false
   )
   &SpawnBurst_Instantaneous(Spawn Count=2, Use Spawn Probability=false, UseLoopCountLimit=true)
   /SetVariables(Emitter.SideLerp:NiagaraFloat=&Abs_Float(Float:NiagaraFloat=&DotProduct(Vector A=&ReturnCameraVector(Camera Vectors=::NewEnumerator1, Coordinate Space=::Simulation), Vector B=<1 0 0>))) {isSetParameters=true, module=SetVariables_#4}
  ]
  particleSpawnScript [
   &PartitionParticles(ID to Partition:NiagaraInt32=&ReturnExecIndex())
   &InitializeParticle(
    Color Mode=::NewEnumerator0
    Position Offset=&Add_Vector(
     A=&MergeHelper_VectorSwitch(0=<6.9 -7.7 9.9>, 1=<-6.9 -7.7 9.9>, 2=?, 3=?, Index=$Particles.PartitionParticles.Partition, Num=::NewEnumerator0)
     B=$System.OffsetByGender
    )
    UsePositionOffset=true
    Write Material Random=true
   )
   /SetVariables() {isSetParameters=true, module=SetVariables_#5}
  ]
  particleUpdateScript [
   &ParticleState(Kill Particles When Lifetime Has Elapsed=false, Let Infinitely Lived Particles Die When Emitter Deactivates=true)
   /SetVariables(Particles.MainColor:LinearColor=&MultiplyLinearColor(
     Linear Color=<rgba 1 0.3278 0.7529 1>
     Scale Factor=&MakeVector4FromVectorAndFloat(
      "Vector (RGB)"=&VectorFromFloat(Value=&Waveform(
        Waveform Count=::NewEnumerator1
        Global Input Position=$Particles.PartitionParticles.NormalizedPartition
        "Waveform [1]"=::NewEnumerator12
        "[1] Frequency"=6
        "[1] Amplitude Min/Max"=<0.35 1>
        "[1] Bias"=0.2
        "[2] Average"=false
        "[2] Input Position"=0
        "[2] Phase"=$Emitter.Age
        "[2] Phase Scale"=1
        "[2] Compound Frequency"=2
        "[2] Frequency"=1.2
        "[2] Amplitude Min/Max"=<0 0.2>
        "[2] Scale"=1
        "[2] Bias"=0
        "[2] Exponent"=1
        "[2] Pulse Duration"=0.1
        "[2] Clip Triangle Min/Max"=<0 0>
        "[2] Random Seed"=0
       ))
      "Float (Alpha)"=0.7
     )
    )) {isSetParameters=true, module=SetVariables_#6}
   /SetVariables(
    Particles.SoftGlowCameraOffset=3
    Particles.SoftGlowColor:LinearColor=&MultiplyLinearColor(
     Linear Color=$StackContext.MainColor
     Scale Factor=&MakeVector4_FromFloat(Float=&MergeHelper_FloatSwitch(0=0.4, 1=0.4, 2=0.2, 3=0.2, 4=0.5, Index=$Particles.PartitionParticles.Partition, Num=::NewEnumerator3))
    )
    Particles.SoftGlowSize:Vector2f=&Vector2DFromFloat(Value=&MergeHelper_FloatSwitch(0=13, 1=13, 2=13, 3=13, 4=13, Index=$Particles.PartitionParticles.Partition, Num=::NewEnumerator3))
   ) {isSetParameters=true, module=SetVariables_#7}
   &CameraOffset()
  ]
  rendererClasses [&NiagaraSpriteRendererProperties]
  renderers [{rendererClass=&NiagaraSpriteRendererProperties}]
 }
 FN_Debug {
  bEnabled=false
  emitterUpdateScript [
   &EmitterState(Life Cycle Mode=::NewEnumerator1, Loop Behavior=::NewEnumerator1, Loop Duration=1, UseLoopDelay=false)
   &SpawnBurst_Instantaneous(Spawn Count=1, Use Spawn Probability=false)
  ]
  particleSpawnScript [
   /Niagara/Modules/Spawn/Initialization/InitializeParticle(
    Lifetime=5
    Module.Write Color=true
    Module.Write Lifetime=true
    Module.Write Mass=true
    Module.Write Material Random=true
    Module.Write Position=false
    Module.Write Scale=false
    Module.Write SpriteRotation=false
    Module.Write SpriteSize=true
    Position=&SimulationPosition()
    Sprite Size=<100 100>
   )
   &SystemLocation(MaskSpawnGroup=false, Offset=<-100 0 0>)
  ]
  particleUpdateScript [
   &ParticleState()
   /SetVariables(Particles.Debug:NiagaraFloat=&MakeCustomFloatFromBool(Bool=$User.bIsMale, True Float=1, False Float=0)) {isSetParameters=true}
   &DynamicMaterialParameters(
    Index 0 Param 1=$Particles.Debug
    Module.Param0WriteEnabled=true
    Module.Param1WriteEnabled=true
    Module.Param2WriteEnabled=true
    Module.Param3WriteEnabled=true
   )
  ]
  rendererClasses [&NiagaraSpriteRendererProperties]
  renderers [{rendererClass=&NiagaraSpriteRendererProperties}]
 }
 GlowsAndLights {
  emitterUpdateScript [
   &EmitterState(
    Inactive Response=::NewEnumerator1
    Life Cycle Mode=::NewEnumerator1
    Loop Behavior=::NewEnumerator1
    Loop Duration Mode=::NewEnumerator1
    UseLoopDelay=false
   )
   &SpawnBurst_Instantaneous(Spawn Count=5, Use Spawn Probability=false, UseLoopCountLimit=true)
   /SetVariables(Emitter.SideLerp:NiagaraFloat=&Abs_Float(Float:NiagaraFloat=&DotProduct(Vector A=&ReturnCameraVector(Camera Vectors=::NewEnumerator1, Coordinate Space=::Simulation), Vector B=<1 0 0>))) {isSetParameters=true, module=SetVariables_#4}
  ]
  particleSpawnScript [
   &PartitionParticles(ID to Partition:NiagaraInt32=&ReturnExecIndex(), Number Of Partitions=5)
   /Niagara/Modules/Spawn/Initialization/V2/InitializeParticle(
    Color Mode=::NewEnumerator0
    Position Offset=&Add_Vector(
     A=&MergeHelper_VectorSwitch(
      0=<3.6 -13.1 -16.1>
      1=<-3.6 -13.1 -16.1>
      2=<6.8 -7.1 -17.6>
      3=<-6.8 -7.1 -17.6>
      4=<0 -6 -20>
      Index=$Particles.PartitionParticles.Partition
      Num=::NewEnumerator3
     )
     B=$System.OffsetByGender
    )
    UsePositionOffset=true
    Write Material Random=true
   )
   /SetVariables() {isSetParameters=true, module=SetVariables_#5}
  ]
  particleUpdateScript [
   &ParticleState(Kill Particles When Lifetime Has Elapsed=false, Let Infinitely Lived Particles Die When Emitter Deactivates=true)
   /SetVariables(Particles.MainColor:LinearColor=&MultiplyLinearColor(
     Linear Color=<rgba 1 0.3278 0.7529 1>
     Scale Factor=&MakeVector4FromVectorAndFloat(
      "Vector (RGB)"=&VectorFromFloat(Value=&Waveform(
        Waveform Count=::NewEnumerator1
        Global Input Position=$Particles.PartitionParticles.NormalizedPartition
        "Waveform [1]"=::NewEnumerator12
        "[1] Frequency"=6
        "[1] Amplitude Min/Max"=<0.35 1>
        "[1] Bias"=0.2
        "[2] Average"=false
        "[2] Input Position"=0
        "[2] Phase"=$Emitter.Age
        "[2] Phase Scale"=1
        "[2] Compound Frequency"=2
        "[2] Frequency"=1.2
        "[2] Amplitude Min/Max"=<0 0.2>
        "[2] Scale"=1
        "[2] Bias"=0
        "[2] Exponent"=1
        "[2] Pulse Duration"=0.1
        "[2] Clip Triangle Min/Max"=<0 0>
        "[2] Random Seed"=0
       ))
      "Float (Alpha)"=0.7
     )
    )) {isSetParameters=true, module=SetVariables_#6}
   /SetVariables(Particles.BottomFade:NiagaraFloat=&RemapRange(
     Input Value To Remap=&ClampFloat(Float=&DotProduct(Vector A=&ReturnCameraVector(Camera Vectors=::NewEnumerator1, Coordinate Space=::Simulation), Vector B=<0 0 1>))
     Input Min=0.6
     Input Max=1
     Output Min=1
     Output Max=0
    )) {isSetParameters=true}
   /SetVariables(
    Particles.SoftGlowCameraOffset=3
    Particles.SoftGlowColor:LinearColor=&MultiplyLinearColor(
     Linear Color=$StackContext.MainColor
     Scale Factor=&Multiply_Vector4(
      A=&MakeVector4_FromFloat(Float=&MergeHelper_FloatSwitch(
        0=0.2
        1=0.2
        2=0.2
        3=0.2
        4=&Multiply_Float(
         A=&FNTimeOfDayScalar(Night Scale=0.5, Morning Scale=0.6, Day Scale=0.6, Evening Scale=0.6)
         B=&Lerp_Float(A=2, B=1, Alpha=$StackContext.BottomFade)
        )
        Index=$Particles.PartitionParticles.Partition
        Num=::NewEnumerator3
       ))
      B=<1 1 1 0.8>
     )
    )
    Particles.SoftGlowSize:Vector2f=&Vector2DFromFloat(Value=&MergeHelper_FloatSwitch(
      0=10
      1=10
      2=13
      3=13
      4=&Lerp_Float(A=13, B=22, Alpha=$StackContext.BottomFade)
      Index=$Particles.PartitionParticles.Partition
      Num=::NewEnumerator3
     ))
   ) {isSetParameters=true, module=SetVariables_#7}
   /SetVariables(
    Particles.CapsuleGlowCameraOffset=4
    Particles.CapsuleGlowColor:LinearColor=&MultiplyLinearColor(Linear Color=$StackContext.MainColor, Scale Factor=<1 1 1 0.15>)
    Particles.CapsuleSpriteSize:Vector2f=&Vector2DFromFloat(Value=&MergeHelper_FloatSwitch(
      0=6
      1=6
      2=&Lerp_Float(A=6, B=12, Alpha=$Emitter.SideLerp)
      3=&Lerp_Float(A=6, B=12, Alpha=$Emitter.SideLerp)
      4=5
      Index=$Particles.PartitionParticles.Partition
      Num=::NewEnumerator3
     ))
   ) {isSetParameters=true}
   /SetVariables(Particles.CapsuleSpriteAlignment:Vector3f=&TransformVector(
     Vector=&RotateVectorByQuaternion(
      Vector To Rotate=<1 0 0>
      Quaternion=/Game/Effects/Niagara/Modules/MergeEmitters/MergeHelper_QuatSwitch(
       0:Quat4f=&MakeQuaternion(
        Quaternion From=2
        Axis=<1 0 0>
        Angle=0
        X=-82.8382
        Y=59.77
        Z=-37.7581
        From=?
        To=?
        Angle Type=::NewEnumerator0
        Coordinate Space=::Local
        Normalize Vectors=false
       )
       1:Quat4f=&MakeQuaternion(
        Quaternion From=2
        Axis=<1 0 0>
        Angle=0
        X=82.8382
        Y=-59.77
        Z=37.7581
        From=?
        To=?
        Angle Type=::NewEnumerator0
        Coordinate Space=::Local
        Normalize Vectors=false
       )
       2:Quat4f=&MakeQuaternion(
        Quaternion From=2
        Axis=<1 0 0>
        Angle=0
        X=-85.6208
        Y=54.0221
        Z=1.9109
        From=?
        To=?
        Angle Type=::NewEnumerator0
        Coordinate Space=::Local
        Normalize Vectors=false
       )
       3:Quat4f=&MakeQuaternion(
        Quaternion From=2
        Axis=<1 0 0>
        Angle=0
        X=85.6208
        Y=-54.0221
        Z=-1.9109
        From=?
        To=?
        Angle Type=::NewEnumerator0
        Coordinate Space=::Local
        Normalize Vectors=false
       )
       4:Quat4f=?
       5:Quat4f=?
       6:Quat4f=?
       7:Quat4f=?
       8:Quat4f=?
       9:Quat4f=?
       Index:NiagaraInt32=$Particles.PartitionParticles.Partition
       Num:MergeHelper_SwitcherEnum=::NewEnumerator2
      )
     )
     Destination Space=::World
    )) {isSetParameters=true}
   /SetVariables(Particles.CapsuleSpriteDynamic:Vector4f=&MakeFloat4FromVectorAndFloat(
     "Vector (RGB)"=&NormalizeVector(Vector To Normalize=$Particles.CapsuleSpriteAlignment, Safe Normalize=true, Safe Fallback Threshold=0, Safe Fallback Vector=<0 0 0.001>)
     "Float (Alpha)"=&MergeHelper_FloatSwitch(0=11, 1=11, 2=12, 3=12, 4=5, Index=$Particles.PartitionParticles.Partition, Num=::NewEnumerator3)
    )) {isSetParameters=true}
   /SetVariables(Particles.MeshOrientation=&OrientMeshToVector(
     Look At Direction=&NormalizeVector(
      Vector To Normalize=&TransformVector(Vector=$Particles.CapsuleSpriteAlignment, Source Space=::World, Destination Space=::Local)
      Safe Normalize=true
      Safe Fallback Threshold=0
      Safe Fallback Vector=<0 0 0.001>
     )
     Reference Axis Vector=<1 0 0>
    )) {enabled=false, isSetParameters=true}
   &CameraOffset(Camera Offset Amount=5)
  ]
  rendererClasses [&NiagaraSpriteRendererProperties]
  renderers [
   {rendererClass=&NiagaraSpriteRendererProperties}
   {rendererClass=&NiagaraSpriteRendererProperties, rendererIndex=1}
  ]
 }
 Minimal003 {
  emitterSpawnScript [/SetVariables(Emitter.RibbonLength:NiagaraFloat=&Multiply_Float(A=8, B=-1), Emitter.StartPos:Vector3f=<0 -6.5 -20>) {isSetParameters=true}]
  emitterUpdateScript [
   &EmitterState(
    Inactive Response=::NewEnumerator1
    Life Cycle Mode=::NewEnumerator1
    Loop Behavior=::NewEnumerator1
    Loop Duration Mode=::NewEnumerator1
    UseLoopDelay=false
   )
   &SpawnBurst_Instantaneous(Spawn Count=6, Use Spawn Probability=false, UseLoopCountLimit=false)
  ]
  particleSpawnScript [
   &InitializeParticle(
    Color=<rgba 1.3 1.3 1.3 1>
    Lifetime=5
    Mesh Scale Mode=::NewEnumerator3
    Mesh Uniform Scale=2
    Sprite Size Mode=::NewEnumerator3
    Uniform Sprite Size=3
    UsePositionOffset=true
    Write Material Random=true
   )
   /SetVariables(Particles.ParticleOrder:NiagaraFloat=&ReturnNormalizedExecIndex()) {isSetParameters=true}
   /SetVariables() {isSetParameters=true}
  ]
  particleUpdateScript [
   &ParticleState(
    Kill Particles When Lifetime Has Elapsed=false
    Let Infinitely Lived Particles Die When Emitter Deactivates=true
    Loop Particles Lifetime=true
   )
   /SetVariables(Particles.GlitchCurve:NiagaraFloat=&Waveform("Waveform [1]"=::NewEnumerator0, "[1] Phase Scale"=0.1, "[1] Frequency"=1, "[1] Amplitude Min/Max"=<0 1>, "[1] Exponent"=10)) {isSetParameters=true}
   /SetVariables(Particles.TimeOffset:NiagaraFloat=&Add_Float(
     A=$Particles.TimeOffset
     B=&Multiply_Float(
      A=$Engine.DeltaTime
      B=&Multiply_Float(
       A=&Lerp_Float(A=1, B=1.5, Alpha=$StackContext.GlitchCurve)
       B=&Lerp_Float(A=1, B=2, Alpha=$System.Timeline.Time)
      )
     )
    )) {isSetParameters=true}
   /SetVariables(Particles.Position=&AddVectorToPosition(
     Position=&Lerp_Vector(
      A=$Emitter.StartPos
      B=&Add_Vector(
       A=$Emitter.StartPos
       B=&Multiply_VectorByFloat(
        Vector=&RotateVectorByQuaternion(
         Vector To Rotate=&TransformVector(
          Vector=&MakeCustomVectorFromBool(Bool=$User.IsSkydiving, True Vector=<0 1 1>, False Vector=<0 0 1>)
          Source Space=::World
          Destination Space=::Local
         )
         Quaternion=$Engine.Owner.Rotation
        )
        Float=$Emitter.RibbonLength
       )
      )
      Alpha=$Particles.ParticleOrder
     )
     Vector=$System.OffsetByGender
    )) {isSetParameters=true}
   /SetVariables(Particles.BottomFade:NiagaraFloat=&RemapRange(
     Input Value To Remap=&ClampFloat(Float=&DotProduct(Vector A=&ReturnCameraVector(Camera Vectors=::NewEnumerator1, Coordinate Space=::Simulation), Vector B=<0 0 1>))
     Input Min=0.6
     Input Max=1
     Output Min=1
     Output Max=0
    )) {isSetParameters=true}
   /SetVariables(Particles.Position=&AddVectorToPosition(
     Position=$StackContext.Position
     Vector=&MakeVector(
      X=&Waveform(
       Global Amplitude Scale=&FloatFromCurve(FloatCurve=DataInterface() {propertyValues {Curve {keys [<at 0 0>,<at 1 1>]}}}, CurveIndex=$Particles.ParticleOrder)
       "Waveform [1]"=::NewEnumerator0
       "[1] Input Position"=$Particles.ParticleOrder
       "[1] Phase Scale"=0.5
       "[1] Frequency"=1
       "[1] Amplitude Min/Max"=<-1 1>
      )
      Y=&MakeFloatFromVector(
       Vector=&MakeVector(
        X=&Waveform(
         Global Amplitude Scale=&FloatFromCurve(FloatCurve=DataInterface() {propertyValues {Curve {keys [<at 0 0>,<at 1 1>]}}}, CurveIndex=$Particles.ParticleOrder)
         "Waveform [1]"=::NewEnumerator1
         "[1] Input Position"=$Particles.ParticleOrder
         "[1] Phase Scale"=0.35
         "[1] Frequency"=1
         "[1] Amplitude Min/Max"=<-1 1>
         "[1] Scale"=0.5
        )
        Y=0
        Z=0
       )
       Channel=::NewEnumerator0
      )
      Z=0
     )
    )) {isSetParameters=true}
   &InterpolateOverTime(
    Rate Of Change:NiagaraFloat=&MakeCustomFloatFromBool(Bool=$User.IsSkydiving, True Float=25, False Float=20)
    Target Position Value:NiagaraPosition=&TransformPosition(Position=$Particles.Position, Source Space=::Local, Destination Space=::World)
    Variable Type=::NewEnumerator6
   )
   /SetVariables(Particles.Flicker:NiagaraFloat=&Lerp_Float(
     A=&Waveform("Waveform [1]"=::NewEnumerator8, "[1] Frequency"=12, "[1] Amplitude Min/Max"=<0.4 0.5>)
     B=&Waveform("Waveform [1]"=::NewEnumerator8, "[1] Frequency"=12, "[1] Amplitude Min/Max"=<0.8 1>)
     Alpha=$System.Timeline.Time
    )) {isSetParameters=true}
   /SetVariables(Particles.Position=&Lerp_Vector(
     A=Linked() {linkedVariable {name=StackContext.Position, type=NiagaraPosition}}
     B=&MakeCustomVectorFromBool(
      Bool=$User.IsSkydiving
      True Vector=&MakeVector(
       X=&ClampFloat(
        Float=&MakeFloatFromVector(
         Vector=&TransformPosition(Position=$StackContext.InterpolateOverTime.InterpolateOverTime_Position.Moving Average, Source Space=::World, Destination Space=::Local)
         Channel=::NewEnumerator0
        )
        Min=-200
        Max=200
       )
       Y=&MakeFloatFromVector(Vector=Linked() {linkedVariable {name=StackContext.Position, type=NiagaraPosition}}, Channel=::NewEnumerator1)
       Z=&MakeFloatFromVector(Vector=Linked() {linkedVariable {name=StackContext.Position, type=NiagaraPosition}}, Channel=::NewEnumerator2)
      )
      False Vector=&MakeVector(
       X=&ClampFloat(
        Float=&MakeFloatFromVector(
         Vector=&TransformPosition(Position=$StackContext.InterpolateOverTime.InterpolateOverTime_Position.Moving Average, Source Space=::World, Destination Space=::Local)
         Channel=::NewEnumerator0
        )
        Min=-100
        Max=100
       )
       Y=&ClampFloat(
        Float=&MakeFloatFromVector(
         Vector=&TransformPosition(Position=$StackContext.InterpolateOverTime.InterpolateOverTime_Position.Moving Average, Source Space=::World, Destination Space=::Local)
         Channel=::NewEnumerator1
        )
        Min=-7
        Max=80
       )
       Z=&Lerp_Float(
        A=&MakeFloatFromVector(
         Vector=&TransformPosition(Position=$StackContext.InterpolateOverTime.InterpolateOverTime_Position.Moving Average, Source Space=::World, Destination Space=::Local)
         Channel=::NewEnumerator2
        )
        B=&MakeFloatFromVector(Vector=Linked() {linkedVariable {name=StackContext.Position, type=NiagaraPosition}}, Channel=::NewEnumerator2)
        Alpha=0.8
       )
      )
     )
     Alpha=&Power(a=$Particles.ParticleOrder, b=2)
    )) {isSetParameters=true}
   /SetVariables(Particles.RibbonWidth=&Multiply_Float(
     A=&FloatFromCurve(FloatCurve=DataInterface() {propertyValues {Curve {keys [<at 0 0.5>,<at 0.3 1>,<at 1 0.2>]}}}, CurveIndex=$Particles.ParticleOrder)
     B=&Multiply_Float(A=6, B=&Lerp_Float(A=0.7, B=0.8, Alpha=&ClampFloat(Float=$Particles.GlitchCurve)))
    )) {isSetParameters=true}
   /SetVariables(
    Particles.RibbonGlowColor:LinearColor=&Multiply_LinearColorByFloat(Linear Color=<rgba 1.8648 0.4 2 0.8>, Float=&Multiply_Float(A=1, B=$StackContext.BottomFade))
    Particles.RibbonGlowWidth:NiagaraFloat=&Multiply_Float(A=$StackContext.RibbonWidth, B=2)
   ) {isSetParameters=true}
   &DynamicMaterialParameters(
    Index 0 Param 1=1
    Index 1 Param 1=1
    Index 1 Param 2=$Particles.TimeOffset
    Index 1 Param 3=0.7
    Index 1 Param 4=0.5
    Module.Index0Float0Write=true
    Module.Index0Float1Write=true
    Module.Index0Float2Write=true
    Module.Index0Float3Write=true
    Module.Index1Float0Write=false
    Module.Index1Float1Write=true
    Module.Index1Float2Write=true
    Module.Index1Float3Write=true
    Write Parameter Index 1=::NewEnumerator1
   )
   &ScaleColor(
    Scale Alpha=$Particles.BottomFade
    Scale RGB=&VectorFromFloat(Value=&Add_Float(A=$Particles.Flicker, B=1))
    ScaleA=true
    ScaleRGB=true
   )
  ]
  rendererClasses [&NiagaraRibbonRendererProperties, &NiagaraMeshRendererProperties]
  renderers [
   {rendererClass=&NiagaraRibbonRendererProperties}
   {rendererClass=&NiagaraRibbonRendererProperties, rendererIndex=1}
   {rendererClass=&NiagaraMeshRendererProperties, rendererIndex=2}
  ]
 }
}
summary {
 systemName=NS_BackPack_AlderOptic
 emitters [
  {emitterName=GlowsAndLights, rendererClasses [&NiagaraSpriteRendererProperties]}
  {emitterName=Minimal003, rendererClasses [&NiagaraRibbonRendererProperties, &NiagaraMeshRendererProperties]}
  {emitterName=EarGlows, rendererClasses [&NiagaraSpriteRendererProperties]}
  {bEnabled=false, emitterName=FN_Debug, rendererClasses [&NiagaraSpriteRendererProperties]}
 ]
 userVariables [
  User.bIsMale {type=NiagaraBool}
  User.IsSkydiving {type=NiagaraBool}
 ]
}
