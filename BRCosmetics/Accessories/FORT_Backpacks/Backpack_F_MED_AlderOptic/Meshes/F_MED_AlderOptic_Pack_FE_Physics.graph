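# Physics asset for the F_MED_AlderOptic backpack cosmetic, providing collision bodies and joints for a charm hanging from the pack.
# Three capsules approximate the pack, while smaller capsules cover the charm's base and four dynamic links, with two on the first link.
# The links chain from base_charm_1 with tight swing limits and zero-degree twist limits.
# The base's constraint to the pack leaves swing and twist free.
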PhysicsAsset /BRCosmetics/Accessories/FORT_Backpacks/Backpack_F_MED_AlderOptic/Meshes/F_MED_AlderOptic_Pack_FE_Physics
bodies {
 base_charm_1 [{shapeType=Capsule, length=1, radius=0.5, shapeName=base_charm_1_capsule}]
 dyn_charm_1 [
  {shapeType=Capsule, length=1.3175, radius=0.5, shapeName=dyn_charm_1_capsule_A, center=<1.1013 -0.5323 0>, rotation=<rot 0 90 -90>}
  {shapeType=Capsule, length=1.3175, radius=0.5, shapeName=dyn_charm_1_capsule_B, center=<1.1013 0.627 0>, rotation=<rot 0 90 -90>}
 ]
 dyn_charm_2 [{shapeType=Capsule, length=0.7707, radius=0.5, shapeName=dyn_charm_3_capsule, center=<0.6034 0 0>, rotation=<rot 90 0 0>}]
 dyn_charm_3 [{shapeType=Capsule, length=0.7707, radius=0.5, shapeName=dyn_charm_3_capsule, center=<0.6034 0 0>, rotation=<rot 90 0 0>}]
 dyn_charm_4 [{
   shapeType=Capsule
   length=1.5168
   radius=1.5467
   shapeName=dyn_charm_4_capsule
   center=<2.0344 -0.1962 -0.1323>
   rotation=<rot 0 90.609 85.5217>
  }]
 pack [
  {shapeType=Capsule, length=22.7231, radius=9.5449, shapeName=pack_capsule_A, center=<0.1319 -4.5218 -2.9885>, rotation=<rot -0.2828 0 0>}
  {
   shapeType=Capsule
   length=24.5543
   radius=5.4457
   shapeName=pack_capsule_B
   center=<-6.334 -2.5885 -2.2223>
   rotation=<rot 20.0854 -0.0002 -0.0005>
  }
  {shapeType=Capsule, length=24.7697, radius=5.2303, shapeName=pack_capsule_C, center=<-2.87 -11.4121 0>}
 ]
}
constraints [
 {bone1Name=dyn_charm_1, bone2Name=base_charm_1, swing1LimitDegrees=5, swing2LimitDegrees=10, twistLimitDegrees=0}
 {bone1Name=dyn_charm_2, bone2Name=dyn_charm_1, swing1LimitDegrees=5, swing2LimitDegrees=10, twistLimitDegrees=0}
 {bone1Name=dyn_charm_3, bone2Name=dyn_charm_2, swing1LimitDegrees=5, swing2LimitDegrees=10, twistLimitDegrees=0}
 {bone1Name=dyn_charm_4, bone2Name=dyn_charm_3, swing1LimitDegrees=5, swing2LimitDegrees=10, twistLimitDegrees=0}
 {
  bone1Name=base_charm_1
  bone2Name=pack
  swing1LimitDegrees=45
  swing1Motion=Free
  swing2LimitDegrees=45
  swing2Motion=Free
  twistLimitDegrees=45
  twistMotion=Free
 }
]
