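# Animation Blueprint for the AlderOptic backpack cosmetic, adding secondary motion to the pack.
# The AnimGraph caches a pose with additive pack-bone offsets, then feeds two RigidBody simulations with different physics assets and a four-bone Anim Dynamics charm chain.
# Nested bool blends select among the three results.
# The EventGraph computes gravity and rigid-body settings from a parameter asset, a translation from spine yaw, and an offset by character gender.
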AnimBlueprint /BRCosmetics/Accessories/FORT_Backpacks/Backpack_F_MED_AlderOptic/Meshes/F_MED_AlderOptic_Pack_AnimBP
target_skeleton=&/F_MED_AlderOptic_Pack_Skeleton
graphs [
 AnimGraph {nodes [
   n1 = RigidBody!(ComponentPose=Local To Component(LocalPose=Use cached pose 'mainOffsetCP'()), Alpha, OverrideWorldGravity:Vector="0,980,0") -> (Pose) {
    bClampLinearTranslationLimitToRefPose=true
    bOverrideWorldGravity=true
    overridePhysicsAsset=&/F_MED_AlderOptic_Pack_Physics
    simulationSpace=BaseBoneSpace
    baseBoneRef {boneName=root}
    componentAppliedLinearAccClamp=<4000 4000 400>
    simSpaceSettings {
     maxAngularAcceleration=1
     maxAngularVelocity=3
     maxLinearAcceleration=600
     maxLinearVelocity=400
     velocityScaleZ=0.8
     worldAlpha=1
     externalLinearDragV=<0.5 0.5 0.5>
    }
   }
   n2 = Component To Local(ComponentPose=@n1)
   n3 = Anim Dynamics - Bone: dyn_charm_1!(ComponentPose=Local To Component(LocalPose=Use cached pose 'mainOffsetCP'()), Alpha=1, GravityOverride:Vector="0,0,-980") -> (Pose) {
    angularDampingOverride=1
    angularSpringConstant=9000
    bAngularSpring=true
    bOverrideAngularDamping=true
    bUseGravityOverride=true
    gravityScale=0.35
    simulationSpace=RootRelative
    boundBone {boneName=dyn_charm_1}
    componentAppliedLinearAccClamp=<3000 3000 3000>
    componentLinearAccScale=<1 1 1>
    componentLinearVelScale=<1 1 1>
    gravityOverride {z=-980}
    physicsBodyDefinitions [{
      boundBone {boneName=dyn_charm_1}
      boxExtents=<20 30 30>
      constraintSetup {angularLimitsMax=<0 30 7>, angularLimitsMin=<0 0 -7>, angularTarget=<1 0 0>}
      localJointOffset=<3 0 0>
     }]
    simSpaceSettings {maxAngularAcceleration=1, maxAngularVelocity=6}
   }
   n4 = Anim Dynamics - Bone: dyn_charm_2!(ComponentPose=@n3, Alpha=1, GravityOverride:Vector="0,0,-980") -> (Pose) {
    angularDampingOverride=1
    angularSpringConstant=9000
    bAngularSpring=true
    bOverrideAngularDamping=true
    bUseGravityOverride=true
    gravityScale=0.35
    simulationSpace=RootRelative
    boundBone {boneName=dyn_charm_2}
    componentAppliedLinearAccClamp=<3000 3000 3000>
    componentLinearAccScale=<1 1 1>
    componentLinearVelScale=<1 1 1>
    gravityOverride {z=-980}
    physicsBodyDefinitions [{
      boundBone {boneName=dyn_charm_2}
      boxExtents=<20 30 30>
      constraintSetup {angularLimitsMax=<0 30 10>, angularLimitsMin=<0 -5 -10>, angularTarget=<1 0 0>}
      localJointOffset=<3 0 0>
     }]
    simSpaceSettings {maxAngularAcceleration=1, maxAngularVelocity=6}
   }
   n5 = Anim Dynamics - Bone: dyn_charm_3!(ComponentPose=@n4, Alpha=1, GravityOverride:Vector="0,0,-980") -> (Pose) {
    angularDampingOverride=1
    angularSpringConstant=9000
    bAngularSpring=true
    bOverrideAngularDamping=true
    bUseGravityOverride=true
    gravityScale=0.35
    simulationSpace=RootRelative
    boundBone {boneName=dyn_charm_3}
    componentAppliedLinearAccClamp=<3000 3000 3000>
    componentLinearAccScale=<1 1 1>
    componentLinearVelScale=<1 1 1>
    gravityOverride {z=-980}
    physicsBodyDefinitions [{
      boundBone {boneName=dyn_charm_3}
      boxExtents=<20 30 30>
      constraintSetup {angularLimitsMax=<0 30 10>, angularLimitsMin=<0 -5 -10>, angularTarget=<1 0 0>}
      localJointOffset=<3 0 0>
     }]
    simSpaceSettings {maxAngularAcceleration=1, maxAngularVelocity=6}
   }
   n6 = Anim Dynamics - Bone: dyn_charm_4!(ComponentPose=@n5, Alpha=1, GravityOverride:Vector="0,0,-980") -> (Pose) {
    angularDampingOverride=1
    angularSpringConstant=9000
    bAngularSpring=true
    bOverrideAngularDamping=true
    bUseGravityOverride=true
    gravityScale=0.35
    simulationSpace=RootRelative
    boundBone {boneName=dyn_charm_4}
    componentAppliedLinearAccClamp=<3000 3000 3000>
    componentLinearAccScale=<1 1 1>
    componentLinearVelScale=<1 1 1>
    gravityOverride {z=-980}
    physicsBodyDefinitions [{
      boundBone {boneName=dyn_charm_4}
      boxExtents=<20 30 30>
      constraintSetup {angularLimitsMax=<0 30 10>, angularLimitsMin=<0 -10 -10>, angularTarget=<1 0 0>}
      localJointOffset=<3 0 0>
     }]
    simSpaceSettings {maxAngularAcceleration=1, maxAngularVelocity=6}
   }
   n7 = Component To Local(ComponentPose=@n6)
   n8 = RigidBody!(
    ComponentPose=Local To Component(LocalPose=Use cached pose 'mainOffsetCP'())
    Alpha
    OverrideWorldGravity="0,0,0"
    ComponentLinearAccScale:Vector="100,100,100"
    ComponentLinearVelScale:Vector="100,100,100"
    SimSpaceSettings:Sim Space Settings Structure=(WorldAlpha=0.5,VelocityScaleZ=0,DampingAlpha=1,MaxLinearVelocity=500,MaxAngularVelocity=3,MaxLinearAcceleration=700,MaxAngularAcceleration=1000,ExternalLinearDragV=(X=0.2,Y=0.2,Z=0),ExternalLinearVelocity=(X=0,Y=0,Z=0),ExternalAngularVelocity=(X=0,Y=0,Z=0))
   ) -> (Pose) {
    bClampLinearTranslationLimitToRefPose=true
    bOverrideWorldGravity=true
    overridePhysicsAsset=&/F_MED_AlderOptic_Pack_FE_Physics
    simulationSpace=BaseBoneSpace
    alphaBoolBlend {blendOption=Linear}
    baseBoneRef {boneName=root}
    componentAppliedLinearAccClamp=<3000 3000 3000>
    componentLinearAccScale=<100 100 100>
    componentLinearVelScale=<100 100 100>
    overrideWorldGravity {z=-980}
    simSpaceSettings {
     maxAngularAcceleration=1000
     maxAngularVelocity=3
     maxLinearAcceleration=700
     maxLinearVelocity=500
     velocityScaleZ=0
     worldAlpha=0.5
     externalLinearDragV=<0.2 0.2>
    }
   }
   Output Pose(Result=Blend Poses by bool(BlendPose_0=Component To Local(ComponentPose=@n8), BlendPose_1=Blend Poses by bool(BlendPose_0=@n2, BlendPose_1=@n7)) {blendType=Linear}) {layerGroup=None}
   n9 = "Transform (Modify) Bone - Bone: pack"!(ComponentPose=Local To Component(LocalPose=Input Pose()), Alpha, Translation="0,0,0") -> (Pose) {translationMode=BMM_Additive, translationSpace=BCS_BoneSpace, boneToModify {boneName=pack}}
   Save cached pose 'mainOffsetCP'(Pose=Component To Local(ComponentPose="Transform (Modify) Bone - Bone: pack"(ComponentPose=@n9, Translation="0,0,0") {translationMode=BMM_Additive, translationSpace=BCS_BoneSpace, boneToModify {boneName=pack}})) {cacheName=mainOffsetCP}
  ]}
 EventGraph {nodes [
   Event BlueprintUpdateAnimation()
   > n1 = Sequence()
   > n2 = "Gravity Override State Function (Parameter Asset) (Nativized)"(
    CharacterPartAnimInstance=Self-Reference()
    GameTimeInSeconds=GetGameTimeInSeconds()
    Parameters=&/F_MED_AlderOptic_Pack_Parameters
   )
   > Set GravityOverride(GravityOverride=@n2)
   > Set RBAN Settings Out(RBAN Settings Out=@n2.OutRigidBodyAnimNodeParameters)
   n3 = GetOwnerSkeletalMeshComponent()
   n4 = BreakRotator(InRot=BreakTransform(InTransform=GetDeltaTransformFromRefPose(self=@n3, BoneName=spine_05, BaseName=spine_01)).Rotation)
   Set backpack_translation(execute=@n1.then_1, backpack_translation=MakeVector(Y=MapRangeClamped(Value=@n4.Yaw, InRangeA=0, InRangeB=30)))
   n5 = Event OnCharacterCustomizationLoaded()
   n6 = Get Array Item with EFort Custom Part Type(Array:Array of Custom Character Part Object References=@n5.CustomParts) -> (Output:Custom Character Part Object Reference)
   > n7 = Branch(Condition="Equal (Enum)"(
     A:EFortCustomGender Enum=Get GenderPermitted!(self:Custom Character Part Object Reference=@n6) -> (GenderPermitted:EFortCustomGender Enum=Invalid)
     B:EFortCustomGender Enum=Male
    ))
   > Set offsetByGender(offsetByGender:Vector="0,-1,0") -> (Output_Get:Vector)
   Set offsetByGender(execute=@n7.else, offsetByGender="0,0.2,0")
  ]}
]
variables [
 backpack_translation {type=Vector}
 GravityOverride {type=Vector}
 offsetByGender {type=Vector}
 RBAN Settings Out {type=FortRigidBodyAnimNodeParameters}
]
